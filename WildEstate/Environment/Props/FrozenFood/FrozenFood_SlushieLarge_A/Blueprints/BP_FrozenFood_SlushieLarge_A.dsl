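; Large slushie frozen food prop Blueprint, a child class that adds its own cup and cup glass meshes to inherited slushie behavior.
; Overrides of the shake intensity and slush amount setters call the parent, then write custom primitive data on the cup mesh (index 2 and 1) to feed its material.
; The construction script runs the parent's, then makes both meshes visible.
; BeginPlay, Tick and overlap events are empty.
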(event EventBeginPlay)

(event Collision|EventActorBeginOverlap
  (params (OtherActor :type Actor%20Object%20Reference)))

(event EventTick
  (params (DeltaSeconds :type Float%20%28single-precision%29)))

(fn SetShakeIntensityCustomPrim
  (params (Value :type Float%20%28single-precision%29))
  (|CallParentFunction Value)
  (Rendering|Material|SetCustomPrimitiveDataFloat (Variables|Default|GetSM_FrozenFood_SlushieLargeCup_A) 2 Value))

(fn SetSlushAmountCustomPrim
  (params (Value :type Float%20%28single-precision%29))
  (|CallParentFunction Value)
  (Rendering|Material|SetCustomPrimitiveDataFloat (Variables|Default|GetSM_FrozenFood_SlushieLargeCup_A) 1 Value))

(fn ConstructionScript
  (|CallParentFunction)
  (Rendering|SetVisibility (Variables|Default|GetSM_FrozenFood_SlushieLargeCup_A) true false)
  (Rendering|SetVisibility (Variables|Default|GetSM_FrozenFood_SlushieLargeCupGlass_A) true false))
